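# Animation Blueprint for the CrownOrder backpack cosmetic, adding procedural bob and sway to its crown, scepter and pillow bones.
# Chained cached poses apply base-bone offsets, linear-spring AnimDynamics, additive bone translation, then angular-spring AnimDynamics on scepter and crown.
# The event graph derives in-air flags from falling, skydive and parachute states, and computes time-shifted sine vectors for ambient drift.
# Chair dynamics nodes sit outside the output chain.
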AnimBlueprint /BRCosmetics/Accessories/FORT_Backpacks/Backpack_M_MED_CrownOrder/Meshes/M_MED_CrownOrder_Pack_AnimBP
target_skeleton=&/M_MED_CrownOrder_Pack_Skeleton
graphs [
 AnimGraph {nodes [
   Anim Dynamics - Bone: dyn_chair(Alpha=0.3, ExternalForce="0,0,0") {
    angularDampingOverride=1
    angularSpringConstant=20000
    bLinearSpring=true
    bOverrideAngularDamping=true
    bUseGravityOverride=true
    gravityScale=0.35
    linearSpringConstant=100
    alphaBoolBlend {blendInTime=0.5, blendOption=QuadraticInOut, blendOutTime=2}
    boundBone {boneName=dyn_chair}
    componentAppliedLinearAccClamp=<4000 4000 4000>
    componentLinearAccScale=<0.8 0.8 0.8>
    componentLinearVelScale=<0.6 0.6 0.6>
    physicsBodyDefinitions [{
      boundBone {boneName=dyn_chair}
      boxExtents=<20 30 30>
      constraintSetup {angularLimitsMax=<0 0 0>, angularLimitsMin=<0 0 0>, angularTarget=<1 0 0>, linearAxesMax=<1 1 1>, linearAxesMin=<-1 0 -1>}
      localJointOffset=<0 0 0>
     }]
    relativeSpaceBone {boneName=base_chair}
   }
   "Transform (Modify) Bone - Bone: dyn_chair"!(ComponentPose, bAlphaBoolEnabled:bool=True, Translation="0,0,0", Rotation="0,0,0") -> (Pose) {
    alphaInputType=Bool
    translationMode=BMM_Additive
    translationSpace=BCS_BoneSpace
    alphaBoolBlend {blendInTime=0.2, blendOption=QuadraticInOut, blendOutTime=0.1}
    boneToModify {boneName=dyn_chair}
   }
   Output Pose(Result=Use cached pose 'OutputCache'()) {layerGroup=None}
   Save cached pose 'InputCache'(Pose=Input Pose()) {cacheName=InputCache}
   n1 = "Transform (Modify) Bone - Bone: dyn_crown"!(
    ComponentPose=Local To Component(LocalPose=Use cached pose 'animDynCache'())
    bAlphaBoolEnabled:bool=True
    Translation="0,0,0"
    Rotation="0,0,0"
   ) -> (Pose) {
    alphaInputType=Bool
    translationMode=BMM_Additive
    translationSpace=BCS_BoneSpace
    alphaBoolBlend {blendInTime=0.2, blendOption=QuadraticInOut, blendOutTime=0.1}
    boneToModify {boneName=dyn_crown}
   }
   n2 = "Transform (Modify) Bone - Bone: dyn_scepter"!(ComponentPose=@n1, bAlphaBoolEnabled:bool=True, Translation="0,0,0", Rotation="0,0,0") -> (Pose) {
    alphaInputType=Bool
    translationMode=BMM_Additive
    translationSpace=BCS_BoneSpace
    alphaBoolBlend {blendInTime=0.2, blendOption=QuadraticInOut, blendOutTime=0.1}
    boneToModify {boneName=dyn_scepter}
   }
   n3 = "Transform (Modify) Bone - Bone: dyn_pillow"!(ComponentPose=@n2, bAlphaBoolEnabled:bool=True, Translation="0,0,0", Rotation="0,0,0") -> (Pose) {
    alphaInputType=Bool
    translationMode=BMM_Additive
    translationSpace=BCS_BoneSpace
    alphaBoolBlend {blendInTime=0.2, blendOption=QuadraticInOut, blendOutTime=0.1}
    boneToModify {boneName=dyn_pillow}
   }
   Save cached pose 'AmbientCache'(Pose=Component To Local(ComponentPose=@n3)) {cacheName=AmbientCache}
   n4 = Anim Dynamics - Bone: dyn_scepter!(ComponentPose=Local To Component(LocalPose=Use cached pose 'AmbientCache'()), bAlphaBoolEnabled:bool=True, ExternalForce="0,0,0") -> (Pose) {
    alphaInputType=Bool
    angularDampingOverride=1
    angularSpringConstant=6000
    bAngularSpring=true
    bOverrideAngularDamping=true
    bUseGravityOverride=true
    gravityScale=0.35
    linearSpringConstant=300
    simulationSpace=RootRelative
    alphaBoolBlend {blendInTime=0.2, blendOption=Sinusoidal, blendOutTime=0.2}
    boundBone {boneName=dyn_scepter}
    componentAppliedLinearAccClamp=<100 100 100>
    componentLinearAccScale=<0.8 0.8 0.8>
    componentLinearVelScale=<0.6 0.6 0.6>
    gravityOverride {z=-980}
    physicsBodyDefinitions [{
      boundBone {boneName=dyn_scepter}
      boxExtents=<20 30 30>
      constraintSetup {angularLimitsMax=<40 9 9>, angularLimitsMin=<-12 -9 0>, angularTarget=<1 0 0>}
      localJointOffset=<-1 0 -0.5>
     }]
   }
   n5 = Anim Dynamics - Bone: dyn_crown!(ComponentPose=@n4, bAlphaBoolEnabled:bool=True, ExternalForce="0,0,0") -> (Pose) {
    alphaInputType=Bool
    angularDampingOverride=1
    angularSpringConstant=5000
    bAngularSpring=true
    bOverrideAngularDamping=true
    bUseGravityOverride=true
    gravityScale=0.35
    linearSpringConstant=100
    alphaBoolBlend {blendInTime=0.2, blendOption=QuadraticInOut, blendOutTime=0.2}
    boundBone {boneName=dyn_crown}
    componentAppliedLinearAccClamp=<100 100 100>
    componentLinearAccScale=<0.5 0.5 0.5>
    componentLinearVelScale=<0.4 0.4 0.4>
    gravityOverride {z=-980}
    physicsBodyDefinitions [{
      boundBone {boneName=dyn_crown}
      boxExtents=<15 15 15>
      constraintSetup {angularTargetAxis=AxisZ, twistAxis=AxisZ, angularLimitsMax=<15 15 0>, angularLimitsMin=<-15 -15 0>, angularTarget=<0 0 1>}
      localJointOffset=<0 0 -2>
     }]
    relativeSpaceBone {boneName=base_chair}
   }
   Save cached pose 'OutputCache'(Pose=Component To Local(ComponentPose=@n5)) {cacheName=OutputCache}
   n6 = "Transform (Modify) Bone - Bone: base_crown"!(
    ComponentPose=Local To Component(LocalPose=Use cached pose 'InputCache'())
    bAlphaBoolEnabled:bool=True
    Translation="0,0,5"
    Rotation="-16.299,28.040,-15.345"
   ) -> (Pose) {
    alphaInputType=Bool
    rotationMode=BMM_Additive
    translationMode=BMM_Additive
    translationSpace=BCS_WorldSpace
    alphaBoolBlend {blendInTime=0.8, blendOption=QuadraticInOut, blendOutTime=0.2}
    boneToModify {boneName=base_crown}
    rotation=<rot -16.299 28.04 -15.345>
    translation {z=5}
   }
   n7 = "Transform (Modify) Bone - Bone: base_scepter"!(ComponentPose=@n6, bAlphaBoolEnabled:bool=True, Translation="0,0,5.5", Rotation="11.870,2.283,-23.322") -> (Pose) {
    alphaInputType=Bool
    rotationMode=BMM_Additive
    translationMode=BMM_Additive
    translationSpace=BCS_WorldSpace
    alphaBoolBlend {blendInTime=0.8, blendOption=QuadraticInOut, blendOutTime=0.2}
    boneToModify {boneName=base_scepter}
    rotation=<rot 11.87 2.283 -23.322>
    translation {z=5.5}
   }
   n8 = "Transform (Modify) Bone - Bone: base_pillow"!(ComponentPose=@n7, bAlphaBoolEnabled:bool=True, Translation="0,0,2.2", Rotation="16.474,-1.180,-9.776") -> (Pose) {
    alphaInputType=Bool
    rotationMode=BMM_Additive
    translationMode=BMM_Additive
    translationSpace=BCS_WorldSpace
    alphaBoolBlend {blendInTime=0.8, blendOption=QuadraticInOut, blendOutTime=0.2}
    boneToModify {boneName=base_pillow}
    rotation=<rot 16.474 -1.18 -9.776>
    translation {z=2.2}
   }
   n9 = "Transform (Modify) Bone - Bone: base_crown"!(ComponentPose=@n8, bAlphaBoolEnabled=True, Translation="2,-2.4,0") -> (Pose) {
    alphaInputType=Bool
    translationMode=BMM_Additive
    alphaBoolBlend {blendInTime=0.8, blendOption=QuadraticInOut, blendOutTime=0.2}
    boneToModify {boneName=base_crown}
    translation=<2 -2.4>
   }
   n10 = "Transform (Modify) Bone - Bone: base_scepter"!(ComponentPose=@n9, bAlphaBoolEnabled=True, Translation="-3.781,2.300,0.000") -> (Pose) {
    alphaInputType=Bool
    translationMode=BMM_Additive
    alphaBoolBlend {blendInTime=0.8, blendOption=QuadraticInOut, blendOutTime=0.2}
    boneToModify {boneName=base_scepter}
    translation=<-3.7809 2.3 0>
   }
   Save cached pose 'OffsetCache'(Pose=Component To Local(ComponentPose=@n10)) {cacheName=OffsetCache}
   n11 = Anim Dynamics - Bone: dyn_pillow!(ComponentPose=Local To Component(LocalPose=Use cached pose 'OffsetCache'()), bAlphaBoolEnabled:bool=True, ExternalForce="0,0,0") -> (Pose) {
    alphaInputType=Bool
    angularDampingOverride=1
    angularSpringConstant=20000
    bLinearSpring=true
    bOverrideAngularDamping=true
    bUseGravityOverride=true
    gravityScale=0.35
    linearSpringConstant=200
    alphaBoolBlend {blendInTime=0.5, blendOption=QuadraticInOut, blendOutTime=0.1}
    boundBone {boneName=dyn_pillow}
    componentAppliedLinearAccClamp=<3000 3000 3000>
    componentLinearAccScale=<1 1 1>
    componentLinearVelScale=<1 1 1>
    physicsBodyDefinitions [{
      boundBone {boneName=dyn_pillow}
      boxExtents=<15 15 10>
      constraintSetup {angularLimitsMax=<0 0 0>, angularLimitsMin=<0 0 0>, angularTarget=<1 0 0>, linearAxesMax=<0.5 0.5 2>, linearAxesMin=<-0.5 -0.5 -2>}
      localJointOffset=<0 0 0>
     }]
    relativeSpaceBone {boneName=base_chair}
   }
   n12 = Anim Dynamics - Bone: dyn_scepter(ComponentPose=@n11, bAlphaBoolEnabled=True, ExternalForce="0,0,0") {
    alphaInputType=Bool
    angularDampingOverride=1
    angularSpringConstant=20000
    bLinearSpring=true
    bOverrideAngularDamping=true
    bUseGravityOverride=true
    gravityScale=0.35
    linearSpringConstant=200
    alphaBoolBlend {blendInTime=0.5, blendOption=QuadraticInOut, blendOutTime=0.1}
    boundBone {boneName=dyn_scepter}
    componentAppliedLinearAccClamp=<1000 1000 1000>
    componentLinearAccScale=<1 1 1>
    componentLinearVelScale=<0.5 0.5 0.5>
    physicsBodyDefinitions [{
      boundBone {boneName=dyn_scepter}
      boxExtents=<30 5 5>
      constraintSetup {angularLimitsMax=<0 0 0>, angularLimitsMin=<0 0 0>, angularTarget=<1 0 0>, linearAxesMax=<0.5 0.5 0.5>, linearAxesMin=<-0.5 -0.5 -0.5>}
      localJointOffset=<14 0 0>
     }]
    relativeSpaceBone {boneName=base_chair}
   }
   n13 = Anim Dynamics - Bone: dyn_crown(ComponentPose=@n12, bAlphaBoolEnabled=True, ExternalForce="0,0,0") {
    alphaInputType=Bool
    angularDampingOverride=1
    angularSpringConstant=20000
    bLinearSpring=true
    bOverrideAngularDamping=true
    bUseGravityOverride=true
    gravityScale=0.35
    linearSpringConstant=200
    alphaBoolBlend {blendInTime=0.5, blendOption=QuadraticInOut, blendOutTime=0.1}
    boundBone {boneName=dyn_crown}
    componentAppliedLinearAccClamp=<3000 3000 3000>
    componentLinearAccScale=<1 1 1>
    componentLinearVelScale=<1 1 1>
    physicsBodyDefinitions [{
      boundBone {boneName=dyn_crown}
      boxExtents=<15 15 15>
      constraintSetup {angularLimitsMax=<0 0 0>, angularLimitsMin=<0 0 0>, angularTarget=<1 0 0>, linearAxesMax=<0.5 0.5 2>, linearAxesMin=<-0.5 -0.5 -2>}
      localJointOffset=<0 0 6>
     }]
    relativeSpaceBone {boneName=base_chair}
   }
   Save cached pose 'animDynCache'(Pose=Component To Local(ComponentPose=@n13)) {cacheName=animDynCache}
  ]}
 EventGraph {nodes [
   Event BlueprintUpdateAnimation()
   > n1 = Sequence()
   n2 = Get bIsFalling()
   n3 = Get bIsSkydiveFloating()
   n4 = XOR Boolean!(A:bool=XOR Boolean(A=@n2, B=AND Boolean(A=@n2, B=@n3)), B:bool=AND Boolean(A=@n2, B=Get bIsSkydiveDiveMode())) -> (ReturnValue:bool)
   n5 = AND Boolean(A=@n2, B=OR Boolean(A=Get bParachuteOpenedAndStandingUpright() -> (bParachuteOpenedAndStandingUpright:bool), B=Get bIsParachuteOpened()))
   > n6 = Set isInAir(isInAir:bool=OR Boolean(A=XOR Boolean(A=@n4, B=@n5), B=false)) -> (Output_Get:bool)
   > Set isNotInAir(isNotInAir:bool=NOT Boolean(A=@n6)) -> (Output_Get:bool)
   n7 = Sequence!(execute=@n1.then_1) -> (then_0, then_1, then_2:Exec, then_3:Exec)
   n8 = GetRealTimeSeconds()
   n9 = float + float(A=@n8, B=500)
   n10 = "Sin (Radians)"(A=float * float(A=1.2, B=@n9))
   n11 = "Sin (Radians)"(A=float * float(A=0.9, B=@n9))
   n12 = "Sin (Radians)"(A=float * float(A=0.8, B=@n9))
   > Set AmbientTranslationChair(AmbientTranslationChair:Vector=MakeVector(X=float * float(A=@n11, B=0.2), Y=float * float(A=@n10, B=0.5), Z=float * float(A=@n12, B=0.5) {is=title K2Node_PromotableOperator})) -> (Output_Get:Vector)
   n13 = float + float(A=@n8, B=1600)
   n14 = "Sin (Radians)"(A=float * float(A=1.1, B=@n13) {is=title K2Node_CommutativeAssociativeBinaryOperator})
   n15 = "Sin (Radians)"(A=float * float(A=1.2, B=@n13))
   n16 = "Sin (Radians)"(A=float * float(A=1.4, B=@n13))
   Set AmbientTranslationCrown(
    execute=@n7.then_3
    AmbientTranslationCrown:Vector=MakeVector(X=float * float(A=@n14, B=0.4), Y=float * float(A=@n15, B=0.5), Z=float * float(A=@n16, B=0.2) {is=title K2Node_PromotableOperator})
   ) -> (Output_Get:Vector)
   n17 = float + float(A=@n8, B=1600)
   n18 = "Sin (Radians)"(A=float * float(A=0.8, B=@n17) {is=title K2Node_CommutativeAssociativeBinaryOperator})
   n19 = "Sin (Radians)"(A=float * float(A=1.1, B=@n17))
   n20 = "Sin (Radians)"(A=float * float(A=1.4, B=@n17))
   Set AmbientTranslationPillow(
    execute=@n7.then_2
    AmbientTranslationPillow:Vector=MakeVector(X=float * float(A=@n19, B=0.15), Y=float * float(A=@n18, B=0.4), Z=float * float(A=@n20, B=0.6) {is=title K2Node_PromotableOperator})
   ) -> (Output_Get:Vector)
   n21 = "Sin (Radians)"(A=float * float(A=1.2, B=@n8) {is=title K2Node_CommutativeAssociativeBinaryOperator})
   n22 = "Sin (Radians)"(A=float * float(A=1.4, B=@n8))
   n23 = "Sin (Radians)"(A=float * float(A=1.1, B=@n8))
   Set AmbientTranslationSepter(
    execute=@n7.then_1
    AmbientTranslationSepter:Vector=MakeVector(X=float * float(A=@n22, B=0.5), Y=float * float(A=@n21, B=0.2), Z=float * float(A=@n23, B=0.6) {is=title K2Node_PromotableOperator})
   ) -> (Output_Get:Vector)
  ]}
]
variables [
 AmbientTranslationChair {type=Vector}
 AmbientTranslationCrown {type=Vector}
 AmbientTranslationPillow {type=Vector}
 AmbientTranslationSepter {type=Vector}
 isInAir {type=boolean}
 isNotInAir {type=boolean}
]
